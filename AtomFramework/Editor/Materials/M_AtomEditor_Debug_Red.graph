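# Solid red debug material for the AtomFramework editor, judging by its name and folder.
# It uses the Unlit shading model, so scene lighting does not affect it.
# Its graph holds a single constant red colour vector, with no textures or parameters.
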Material /AtomFramework/Editor/Materials/M_AtomEditor_Debug_Red
graph {
 properties {pixelDepthOffsetMode=PDOM_Legacy, shadingModel=MSM_Unlit, usedShadingModels=Unlit}
 nodes [Constant3Vector() {constant=<rgba 1 0 0 0>}]
}
