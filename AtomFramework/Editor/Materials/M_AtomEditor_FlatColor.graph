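# M_AtomEditor_FlatColor is an AtomFramework editor material, named and set up as a flat color for tinting.
# Its graph holds a single Color vector parameter, green by default, that material instances can override; the output it drives is not shown.
# It uses Default Lit shading and is flagged for use with instanced static meshes.
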Material /AtomFramework/Editor/Materials/M_AtomEditor_FlatColor
graph {
 properties {bUsedWithInstancedStaticMeshes=true, pixelDepthOffsetMode=PDOM_Legacy, usedShadingModels=Default Lit}
 nodes [VectorParameter() {parameterName=Color, defaultValue=<rgba 0 1 0 1>}]
}
