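# Translucent, two-sided editor material in AtomFramework, flagged for editor compositing, skeletal meshes and mesh deformers.
# It uses Thin Translucent shading with per-pixel surface lighting, allows custom depth writes and casts no ray-traced shadows.
# A scalar Opacity parameter feeds transmittance color.
# An orange-default Color parameter and a slightly scaled vertex normal, resembling a small surface offset, are also present without shown destinations.
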Material /AtomFramework/Editor/Materials/M_AtomEditor_Translucent
graph {
 properties {
  allowTranslucentCustomDepthWrites=true
  bCastRayTracedShadows=false
  bUseTranslucencyVertexFog=false
  bUsedWithEditorCompositing=true
  bUsedWithMeshDeformer=true
  bUsedWithSkeletalMesh=true
  blendMode=BLEND_Translucent
  pixelDepthOffsetMode=PDOM_Legacy
  shadingModel=MSM_ThinTranslucent
  translucencyLightingMode=TLM_SurfacePerPixelLighting
  twoSided=true
  usedShadingModels=Thin Translucent
 }
 nodes [
  Multiply(A=VertexNormalWS()) {constB=0.001}
  ThinTranslucentMaterialOutput!(TransmittanceColor=ScalarParameter() {defaultValue=0.3, parameterName=Opacity}, SurfaceCoverage)
  VectorParameter() {parameterName=Color, defaultValue=<rgba 0.8139 0.4692 0 1>}
 ]
}
